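# Physics asset for the Vampire two-handed pickaxe skeletal mesh, a Battle Royale melee weapon cosmetic.
# It holds a single capsule body on the Melee_TwoHanded_Handle bone, rotated and offset along the handle to give the pickaxe a rough collision volume.
# With no other bodies and no constraints, the asset defines no articulated or jointed parts.
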PhysicsAsset /BRCosmetics/Weapons/FORT_Melee/Pickaxe_Vampire/Meshes/SK_Pickaxe_Vampire_PhysicsAsset
bodies {Melee_TwoHanded_Handle [{shapeType=Capsule, length=77.9166, radius=52.7077, center=<0.4259 -1.046 25.1508>, rotation=<rot 0.0675 -90 -0.0029>}]}
constraints []
